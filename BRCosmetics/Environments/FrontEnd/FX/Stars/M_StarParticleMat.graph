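# Unlit translucent sprite-particle material for star particles in the front-end environment.
# Particle color RGB feeds the shared emissive adjustments function.
# A soft sphere mask, sized by particle alpha times the MaxStarSize dynamic parameter, forms a dimmed glow.
# That glow is added to the star icon mask texture, clamped, then multiplied by particle alpha, so alpha drives both glow size and fade.
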Material /BRCosmetics/Environments/FrontEnd/FX/Stars/M_StarParticleMat
graph {
 properties {
  bUsedWithParticleSprites=true
  blendMode=BLEND_Translucent
  pixelDepthOffsetMode=PDOM_Legacy
  shadingModel=MSM_Unlit
  usedShadingModels=Unlit
 }
 nodes [
  n1 = ParticleColor()
  MaterialFunctionCall(Emissive=@n1.RGB) {materialFunction=&MF_FX_EmissiveAdjustments}
  n2 = DynamicParameter() {paramNames [MaxStarSize,Param2,Param3,Param4]}
  n3 = Multiply(A=SphereMask(A=TextureCoordinate(), B=Constant() {r=0.5}, Radius=Multiply(A=@n1.A, B=@n2.MaxStarSize)) {hardnessPercent=0}) {constB=0.25}
  n4 = TextureSample() {samplerType=SAMPLERTYPE_Masks, texture=/Game/Environments/FrontEnd/FX/Stars/Textures/T_StarIcon_01}
  Multiply(A=@n1.A, B=Clamp(None=Add(A=@n3, B=@n4.R)))
 ]
}
